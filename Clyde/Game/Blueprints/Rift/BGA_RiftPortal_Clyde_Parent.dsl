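; Parent Blueprint for Clyde rift portals that adds rift cleanup when the player pawn's abilities are cancelled.
; The CancelPlayerPawnAbilities override runs the parent logic, then checks the HandleRiftRemoval scalable float as a bool.
; If that is enabled and the portal's player pawn is valid, it applies the GE_Clyde_RiftedCleanup effect to that pawn as a global environment gameplay effect.
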(fn CancelPlayerPawnAbilities
  (|CallParentFunction)
  (if (ScalableFloat|GetValueasBool (Variables|VFX|GetHandleRiftRemoval))
    (bind _playerpawn (Variables|RiftPortal|GetPlayerPawn)
      (:then
        (Environment|ApplyGlobalEnvironmentGameplayEffecttoActor _playerpawn "/Clyde/GameplayEffects/Rift/GE_Clyde_RiftedCleanup.GE_Clyde_RiftedCleanup_C"))
      (:else))))

(fn ConstructionScript)
